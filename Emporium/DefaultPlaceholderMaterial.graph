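# DefaultPlaceholderMaterial is a translucent, Default Lit material in the Emporium content, named as a default placeholder.
# Its graph holds a single scalar parameter with a preset default and no visible connection to any material input.
# It has no textures, color inputs or other logic.
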Material /Emporium/DefaultPlaceholderMaterial
graph {
 properties {blendMode=BLEND_Translucent, pixelDepthOffsetMode=PDOM_Legacy, usedShadingModels=Default Lit}
 nodes [ScalarParameter() {defaultValue=0.75}]
}
